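; BP_BroadcastPost is an actor in the TimeOfDay folder that switches between a standard and an extreme post-process look.
; Its construction script reads the ExtremePostEnabled flag and enables either the PostProcess or Extreme component, disabling the other.
; BeginPlay, Tick and overlap events are empty, so the Blueprint has no runtime logic of its own.
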(event EventBeginPlay)

(event Collision|EventActorBeginOverlap
  (params (OtherActor :type Actor%20Object%20Reference)))

(event EventTick
  (params (DeltaSeconds :type Float%20%28single-precision%29)))

(fn ConstructionScript
  (if (Variables|Default|GetExtremePostEnabled)
    (Variables|/Script/Engine.PostProcessComponent|SetbEnabled false (Variables|Default|GetPostProcess))
    (Variables|/Script/Engine.PostProcessComponent|SetbEnabled true (Variables|Default|GetExtreme))
    (else
      (Variables|/Script/Engine.PostProcessComponent|SetbEnabled true (Variables|Default|GetPostProcess))
      (Variables|/Script/Engine.PostProcessComponent|SetbEnabled false (Variables|Default|GetExtreme)))))
